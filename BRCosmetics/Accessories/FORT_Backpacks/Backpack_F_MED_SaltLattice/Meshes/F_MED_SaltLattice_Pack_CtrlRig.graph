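# Control Rig for the SaltLattice backpack cosmetic that procedurally sways its branching fishtail bone chain.
# From Begin Execution, a sequence of Item Harmonics units oscillates Fishtail_01 through 09 around Y, with per-bone amplitude and ratio staggering the eased wave.
# Two earlier variants of that sequence remain disconnected from execution.
# An unused HarmonicCustom library function builds eased, noise-offset looping rotations with per-axis disable toggles.
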ControlRigBlueprint /BRCosmetics/Accessories/FORT_Backpacks/Backpack_F_MED_SaltLattice/Meshes/F_MED_SaltLattice_Pack_CtrlRig
graphs [
 HarmonicCustom_ContainedGraph {nodes [
   RigVMFunctionEntryNode()
   > For_Each = RigVMDispatchNode(Array=RigVMVariableNode() {variable=Items To Modify}.Value) {template="DISPATCH_RigVMDispatch_ArrayIterator(in Array,out Element,out Index,out Count,out Ratio)"}
   RigVMFunction_MathBoolNot = &RigVMFunction_MathBoolNot(Value=RigVMVariableNode() {variable=X Rot Disable}.Value)
   Remap = &RigVMFunction_MathDoubleRemap(
    SourceMaximum=1
    SourceMinimum=0
    TargetMaximum=-1
    TargetMinimum=1
    Value=&RigVMFunction_MathBoolToFloat(Value=RigVMVariableNode() {variable=FlipRotation}.Value).Result
   )
   Noise = /Script/RigVM.RigVMFunction_NoiseFloat(
    Maximum=0.1
    Minimum=0
    Frequency=RigVMVariableNode() {variable=Noise Frequency}.Value
    Speed=RigVMVariableNode() {variable=Noise speed}.Value
    Value=RigVMVariableNode() {variable=Noise Value}.Value
   ) {is=struct RigVMUnitNode, template="Noise::Execute(in Value,in Speed,in Frequency,in Minimum,in Maximum,out Result)"}
   RigVMFunction_TimeLoop = /Script/RigVM.RigVMFunction_TimeLoop(Normalize=False, Speed=RigVMVariableNode() {variable=TimeSpeed}.Value, Duration=RigVMVariableNode() {variable=TimeDuration}.Value) {is=struct RigVMUnitNode}
   Add = &RigVMFunction_MathDoubleAdd(B=@Noise.Result, A=@RigVMFunction_TimeLoop.FlipFlop)
   Cast = &RigVMFunction_MathBoolToFloat(Value=RigVMVariableNode() {variable=Enable Negative Angle}.Value)
   n1 = RigVMVariableNode() {variable=Angle}
   RigVMFunction_AnimEasing = /Script/RigVM.RigVMFunction_AnimEasing(
    SourceMaximum=1
    SourceMinimum=0
    Value=@Add.Result
    TargetMinimum=&RigVMFunction_MathDoubleMul(A=&RigVMFunction_MathDoubleMul(B=-1, A=@Cast.Result).Result, B=@n1.Value).Result
    TargetMaximum=@n1.Value
    Type=RigVMVariableNode() {variable=Ease Type Curve}.Value
   ) {is=struct RigVMUnitNode}
   n2 = &RigVMFunction_MathDoubleMul(B=@Remap.Result, A=@RigVMFunction_AnimEasing.Result)
   Multiply = &RigVMFunction_MathDoubleMul(B=&RigVMFunction_MathBoolToFloat(Value=@RigVMFunction_MathBoolNot.Result).Result, A=@n2.Result)
   n3 = &RigVMFunction_MathBoolNot(Value=RigVMVariableNode() {variable=Y Rot Disable}.Value)
   n4 = &RigVMFunction_MathBoolNot(Value=RigVMVariableNode() {variable=Z Rot Disable}.Value)
   RigVMFunction_MathQuaternionFromEuler = &RigVMFunction_MathQuaternionFromEuler(
    Euler=(X=0,Y=0,Z=0)
    Euler.X=@Multiply.Result
    Euler.Y=&RigVMFunction_MathDoubleMul(B=&RigVMFunction_MathBoolToFloat(Value=@n3.Result).Result, A=@n2.Result).Result
    Euler.Z=&RigVMFunction_MathDoubleMul(B=&RigVMFunction_MathBoolToFloat(Value=@n4.Result).Result, A=@n2.Result).Result
   )
   > &RigUnit_ModifyTransforms(
    ItemToModify=((Item=(Type=Bone,Name="upperarm_l"),Transform=(Rotation=(X=0,Y=0,Z=0,W=1),Translation=(X=0,Y=0,Z=0),Scale3D=(X=1,Y=1,Z=1))))
    Weight=1
    WeightMinimum=0
    ItemToModify.0.Item=@For_Each.Element
    ItemToModify.0.Transform.Rotation=@RigVMFunction_MathQuaternionFromEuler.Result
   )
   RigVMFunctionReturnNode(ExecuteContext=@For_Each.Completed)
  ]}
 RigVMFunctionLibrary {nodes [HarmonicCustom = RigVMCollapseNode(
    Angle=15
    Ease Type Curve=SineEaseInOut
    Enable Negative Angle=true
    ExecuteContext=()
    FlipRotation=false
    Items To Modify=((Type=Bone,Name="tail5"),(Type=Bone,Name="tail7"),(Type=Bone,Name="tail9"))
    Noise Frequency=0
    Noise Value=0
    Noise speed=1
    TimeDuration=1
    TimeSpeed=1
    X Rot Disable=false
    Y Rot Disable=false
    Z Rot Disable=false
   )]}
 RigVMModel {nodes [
   /Script/ControlRig.RigUnit_ItemHarmonics(
    ExecutePin=()
    RotationOrder=YZX
    Targets=((Item=(Type=Bone,Name="Fishtail_01"),Ratio=0))
    WaveAmplitude=(X=0,Y=30,Z=0)
    WaveEase=Linear
    WaveFrequency=(X=1,Y=0.6,Z=0.8)
    WaveMaximum=1.5
    WaveMinimum=0.5
    WaveNoise=(X=0,Y=0,Z=0)
    WaveOffset=(X=0,Y=1,Z=2)
    WaveSpeed=(X=1,Y=2,Z=1)
   ) {is=struct RigVMUnitNode}
   > &RigUnit_ItemHarmonics(
    RotationOrder=YZX
    Targets=((Item=(Type=Bone,Name="Fishtail_02"),Ratio=0.1))
    WaveAmplitude=(X=0,Y=20,Z=0)
    WaveEase=Linear
    WaveFrequency=(X=1,Y=0.6,Z=0.8)
    WaveMaximum=1.1
    WaveMinimum=0.3
    WaveNoise=(X=0,Y=0,Z=0)
    WaveOffset=(X=0,Y=1,Z=2)
    WaveSpeed=(X=1,Y=2,Z=1)
   )
   > &RigUnit_ItemHarmonics(
    RotationOrder=YZX
    Targets=((Item=(Type=Bone,Name="Fishtail_03"),Ratio=0))
    WaveAmplitude=(X=0,Y=30,Z=0)
    WaveEase=Linear
    WaveFrequency=(X=1,Y=0.6,Z=0.8)
    WaveMaximum=1.5
    WaveMinimum=0.5
    WaveNoise=(X=0,Y=0,Z=0)
    WaveOffset=(X=0,Y=1,Z=2)
    WaveSpeed=(X=1,Y=4,Z=1)
   )
   > &RigUnit_ItemHarmonics(
    RotationOrder=YZX
    Targets=((Item=(Type=Bone,Name="Fishtail_04"),Ratio=0.1))
    WaveAmplitude=(X=0,Y=20,Z=0)
    WaveEase=Linear
    WaveFrequency=(X=1,Y=0.6,Z=0.8)
    WaveMaximum=1.1
    WaveMinimum=0.3
    WaveNoise=(X=0,Y=0,Z=0)
    WaveOffset=(X=0,Y=1,Z=2)
    WaveSpeed=(X=1,Y=2,Z=1)
   )
   > &RigUnit_ItemHarmonics(
    RotationOrder=YZX
    Targets=((Item=(Type=Bone,Name="Fishtail_07"),Ratio=0))
    WaveAmplitude=(X=0,Y=30,Z=0)
    WaveEase=Linear
    WaveFrequency=(X=1,Y=0.6,Z=0.8)
    WaveMaximum=1.5
    WaveMinimum=0.5
    WaveNoise=(X=0,Y=0,Z=0)
    WaveOffset=(X=0,Y=1,Z=2)
    WaveSpeed=(X=1,Y=4,Z=1)
   )
   > &RigUnit_ItemHarmonics(
    RotationOrder=YZX
    Targets=((Item=(Type=Bone,Name="Fishtail_06"),Ratio=0))
    WaveAmplitude=(X=0,Y=30,Z=0)
    WaveEase=Linear
    WaveFrequency=(X=1,Y=0.6,Z=0.8)
    WaveMaximum=1.5
    WaveMinimum=0.5
    WaveNoise=(X=0,Y=0,Z=0)
    WaveOffset=(X=0,Y=1,Z=2)
    WaveSpeed=(X=1,Y=4,Z=1)
   )
   > &RigUnit_ItemHarmonics(
    RotationOrder=YZX
    Targets=((Item=(Type=Bone,Name="Fishtail_05"),Ratio=0.1))
    WaveAmplitude=(X=0,Y=20,Z=0)
    WaveEase=Linear
    WaveFrequency=(X=1,Y=0.6,Z=0.8)
    WaveMaximum=1.1
    WaveMinimum=0.3
    WaveNoise=(X=0,Y=0,Z=0)
    WaveOffset=(X=0,Y=1,Z=2)
    WaveSpeed=(X=1,Y=2,Z=1)
   )
   &RigUnit_ItemHarmonics(
    ExecutePin=()
    RotationOrder=YZX
    Targets=((Item=(Type=Bone,Name="Fishtail_01"),Ratio=1))
    WaveAmplitude=(X=0,Y=12,Z=0)
    WaveEase=QuadraticEaseInOut
    WaveFrequency=(X=1,Y=1,Z=0.8)
    WaveMaximum=1.3
    WaveMinimum=1.3
    WaveNoise=(X=0,Y=0,Z=0)
    WaveOffset=(X=0,Y=1,Z=2)
    WaveSpeed=(X=1,Y=2,Z=1)
   )
   > &RigUnit_ItemHarmonics(
    RotationOrder=YZX
    Targets=((Item=(Type=Bone,Name="Fishtail_02"),Ratio=0.8))
    WaveAmplitude=(X=0,Y=17,Z=0)
    WaveEase=QuadraticEaseInOut
    WaveFrequency=(X=1,Y=1,Z=0.8)
    WaveMaximum=1.3
    WaveMinimum=1.3
    WaveNoise=(X=0,Y=0,Z=0)
    WaveOffset=(X=0,Y=0,Z=2)
    WaveSpeed=(X=1,Y=2,Z=1)
   )
   > &RigUnit_ItemHarmonics(
    RotationOrder=YZX
    Targets=((Item=(Type=Bone,Name="Fishtail_03"),Ratio=0.6))
    WaveAmplitude=(X=0,Y=22,Z=0)
    WaveEase=QuadraticEaseInOut
    WaveFrequency=(X=1,Y=1,Z=0.8)
    WaveMaximum=1.3
    WaveMinimum=1.3
    WaveNoise=(X=0,Y=0,Z=0)
    WaveOffset=(X=0,Y=0,Z=2)
    WaveSpeed=(X=1,Y=2,Z=1)
   )
   > &RigUnit_ItemHarmonics(
    RotationOrder=YZX
    Targets=((Item=(Type=Bone,Name="Fishtail_04"),Ratio=0.4))
    WaveAmplitude=(X=0,Y=27,Z=0)
    WaveEase=QuadraticEaseInOut
    WaveFrequency=(X=1,Y=1,Z=0.8)
    WaveMaximum=1.3
    WaveMinimum=1.3
    WaveNoise=(X=0,Y=0,Z=0)
    WaveOffset=(X=0,Y=0,Z=2)
    WaveSpeed=(X=1,Y=2,Z=1)
   )
   > &RigUnit_ItemHarmonics(
    RotationOrder=YZX
    Targets=((Item=(Type=Bone,Name="Fishtail_05"),Ratio=0.2))
    WaveAmplitude=(X=0,Y=10,Z=0)
    WaveEase=QuadraticEaseInOut
    WaveFrequency=(X=1,Y=1,Z=0.8)
    WaveMaximum=1.3
    WaveMinimum=1.3
    WaveNoise=(X=0,Y=0,Z=0)
    WaveOffset=(X=0,Y=0,Z=2)
    WaveSpeed=(X=1,Y=2,Z=1)
   )
   > &RigUnit_ItemHarmonics(
    RotationOrder=YZX
    Targets=((Item=(Type=Bone,Name="Fishtail_06"),Ratio=0.2))
    WaveAmplitude=(X=0,Y=15,Z=0)
    WaveEase=QuadraticEaseInOut
    WaveFrequency=(X=1,Y=1,Z=0.8)
    WaveMaximum=1.3
    WaveMinimum=1.3
    WaveNoise=(X=0,Y=0,Z=0)
    WaveOffset=(X=0,Y=0,Z=2)
    WaveSpeed=(X=1,Y=2,Z=1)
   )
   > &RigUnit_ItemHarmonics(
    RotationOrder=YZX
    Targets=((Item=(Type=Bone,Name="Fishtail_07"),Ratio=0.2))
    WaveAmplitude=(X=0,Y=20,Z=0)
    WaveEase=QuadraticEaseInOut
    WaveFrequency=(X=1,Y=1,Z=0.8)
    WaveMaximum=1.3
    WaveMinimum=1.3
    WaveNoise=(X=0,Y=0,Z=0)
    WaveOffset=(X=0,Y=0,Z=2)
    WaveSpeed=(X=1,Y=2,Z=1)
   )
   > &RigUnit_ItemHarmonics(
    RotationOrder=YZX
    Targets=((Item=(Type=Bone,Name="Fishtail_08"),Ratio=0.6))
    WaveAmplitude=(X=0,Y=22,Z=0)
    WaveEase=QuadraticEaseInOut
    WaveFrequency=(X=1,Y=1,Z=0.8)
    WaveMaximum=1.3
    WaveMinimum=1.3
    WaveNoise=(X=0,Y=0,Z=0)
    WaveOffset=(X=0,Y=0,Z=2)
    WaveSpeed=(X=1,Y=2,Z=1)
   )
   > &RigUnit_ItemHarmonics(
    RotationOrder=YZX
    Targets=((Item=(Type=Bone,Name="Fishtail_09"),Ratio=0.4))
    WaveAmplitude=(X=0,Y=27,Z=0)
    WaveEase=QuadraticEaseInOut
    WaveFrequency=(X=1,Y=1,Z=0.8)
    WaveMaximum=1.3
    WaveMinimum=1.3
    WaveNoise=(X=0,Y=0,Z=0)
    WaveOffset=(X=0,Y=0,Z=2)
    WaveSpeed=(X=1,Y=2,Z=1)
   )
   RigUnit_BeginExecution = &RigUnit_BeginExecution()
   &RigUnit_ItemHarmonics(
    RotationOrder=YZX
    Targets=((Item=(Type=Bone,Name="Fishtail_01"),Ratio=1))
    WaveAmplitude=(X=0,Y=12,Z=0)
    WaveEase=QuadraticEaseInOut
    WaveFrequency=(X=1,Y=1,Z=0.8)
    WaveMaximum=1.3
    WaveMinimum=1.3
    WaveNoise=(X=0,Y=0,Z=0)
    WaveOffset=(X=0,Y=1,Z=2)
    WaveSpeed=(X=1,Y=2.05,Z=1)
    ExecutePin=@RigUnit_BeginExecution.ExecutePin
   )
   > &RigUnit_ItemHarmonics(
    RotationOrder=YZX
    Targets=((Item=(Type=Bone,Name="Fishtail_02"),Ratio=0.8))
    WaveAmplitude=(X=0,Y=17,Z=0)
    WaveEase=QuadraticEaseInOut
    WaveFrequency=(X=1,Y=1,Z=0.8)
    WaveMaximum=1.3
    WaveMinimum=1.3
    WaveNoise=(X=0,Y=0,Z=0)
    WaveOffset=(X=0,Y=0,Z=2)
    WaveSpeed=(X=1,Y=2.05,Z=1)
   )
   > &RigUnit_ItemHarmonics(
    RotationOrder=YZX
    Targets=((Item=(Type=Bone,Name="Fishtail_03"),Ratio=0.6))
    WaveAmplitude=(X=0,Y=22,Z=0)
    WaveEase=QuadraticEaseInOut
    WaveFrequency=(X=1,Y=1,Z=0.8)
    WaveMaximum=1.3
    WaveMinimum=1.3
    WaveNoise=(X=0,Y=0,Z=0)
    WaveOffset=(X=0,Y=0,Z=2)
    WaveSpeed=(X=1,Y=2.05,Z=1)
   )
   > &RigUnit_ItemHarmonics(
    RotationOrder=YZX
    Targets=((Item=(Type=Bone,Name="Fishtail_04"),Ratio=0.4))
    WaveAmplitude=(X=0,Y=27,Z=0)
    WaveEase=QuadraticEaseInOut
    WaveFrequency=(X=1,Y=1,Z=0.8)
    WaveMaximum=1.3
    WaveMinimum=1.3
    WaveNoise=(X=0,Y=0,Z=0)
    WaveOffset=(X=0,Y=0,Z=2)
    WaveSpeed=(X=1,Y=2.05,Z=1)
   )
   > &RigUnit_ItemHarmonics(
    RotationOrder=YZX
    Targets=((Item=(Type=Bone,Name="Fishtail_05"),Ratio=0.2))
    WaveAmplitude=(X=0,Y=12,Z=0)
    WaveEase=QuadraticEaseInOut
    WaveFrequency=(X=1,Y=1,Z=0.8)
    WaveMaximum=1.3
    WaveMinimum=1.3
    WaveNoise=(X=0,Y=0,Z=0)
    WaveOffset=(X=0,Y=0,Z=2)
    WaveSpeed=(X=1,Y=2.05,Z=1)
   )
   > &RigUnit_ItemHarmonics(
    RotationOrder=YZX
    Targets=((Item=(Type=Bone,Name="Fishtail_06"),Ratio=0.2))
    WaveAmplitude=(X=0,Y=15,Z=0)
    WaveEase=QuadraticEaseInOut
    WaveFrequency=(X=1,Y=1,Z=0.8)
    WaveMaximum=1.3
    WaveMinimum=1.3
    WaveNoise=(X=0,Y=0,Z=0)
    WaveOffset=(X=0,Y=0,Z=2)
    WaveSpeed=(X=1,Y=2.05,Z=1)
   )
   > &RigUnit_ItemHarmonics(
    RotationOrder=YZX
    Targets=((Item=(Type=Bone,Name="Fishtail_07"),Ratio=0.2))
    WaveAmplitude=(X=0,Y=20,Z=0)
    WaveEase=QuadraticEaseInOut
    WaveFrequency=(X=1,Y=1,Z=0.8)
    WaveMaximum=1.3
    WaveMinimum=1.3
    WaveNoise=(X=0,Y=0,Z=0)
    WaveOffset=(X=0,Y=0,Z=2)
    WaveSpeed=(X=1,Y=2.05,Z=1)
   )
   > &RigUnit_ItemHarmonics(
    RotationOrder=YZX
    Targets=((Item=(Type=Bone,Name="Fishtail_08"),Ratio=0.6))
    WaveAmplitude=(X=0,Y=22,Z=0)
    WaveEase=QuadraticEaseInOut
    WaveFrequency=(X=1,Y=1,Z=0.8)
    WaveMaximum=1.3
    WaveMinimum=1.3
    WaveNoise=(X=0,Y=0,Z=0)
    WaveOffset=(X=0,Y=0,Z=2)
    WaveSpeed=(X=1,Y=2.05,Z=1)
   )
   > &RigUnit_ItemHarmonics(
    RotationOrder=YZX
    Targets=((Item=(Type=Bone,Name="Fishtail_09"),Ratio=0.4))
    WaveAmplitude=(X=0,Y=27,Z=0)
    WaveEase=QuadraticEaseInOut
    WaveFrequency=(X=1,Y=1,Z=0.8)
    WaveMaximum=1.3
    WaveMinimum=1.3
    WaveNoise=(X=0,Y=0,Z=0)
    WaveOffset=(X=0,Y=0,Z=2)
    WaveSpeed=(X=1,Y=2.05,Z=1)
   )
  ]}
]
hierarchy [
 root {loc=<0 0 0>, rot=<0 0 0 1>}
 pack {parent=root, loc=<0 -0.7316 0>, rot=<0 0 0 1>}
 body {parent=pack, loc=<0 -4.3136 -3.5643>, rot=<-0.0955 -0.7006 -0.0955 -0.7006>}
 Fishtail_01 {parent=body, loc=<8.9228 -0.6984 0>, rot=<0 0 0.4278 -0.9039>}
 Fishtail_02 {parent=Fishtail_01, loc=<11.7791 -0.2699 0>, rot=<0 0 -0.1237 -0.9923>}
 Fishtail_03 {parent=Fishtail_02, loc=<8.2661 1.0073 0>, rot=<0 0 -0.259 -0.9659>}
 Fishtail_04 {parent=Fishtail_03, loc=<7.2913 -1.7431 0>, rot=<0 0 0 1>}
 Fishtail_05 {parent=Fishtail_02, loc=<4.2484 -5.7104 0>, rot=<0 0 0.0392 -0.9992>}
 Fishtail_06 {parent=Fishtail_05, loc=<6.179 0.0789 0>, rot=<0 0 -0.2649 -0.9643>}
 Fishtail_07 {parent=Fishtail_06, loc=<3.4544 -0.2038 0>, rot=<0 0 0 1>}
 Fishtail_08 {parent=Fishtail_05, loc=<5.8315 -2.4951 0>, rot=<0 0 -0.0335 -0.9994>}
 Fishtail_09 {parent=Fishtail_08, loc=<5.4253 0 0>, rot=<0 0 -0.1842 -0.9829>}
]
