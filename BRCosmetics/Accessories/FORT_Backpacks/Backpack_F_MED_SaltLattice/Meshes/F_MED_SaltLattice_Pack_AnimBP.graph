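# Animation Blueprint for the SaltLattice back bling, which adds physics sway to its seven-bone fishtail.
# Each update, yaw of the owner mesh's spine_05 relative to spine_01, versus ref pose, drives additive pack rotation and translation.
# The pose takes a Fishtail_01 rotation offset and that pack adjustment, then chained spring AnimDynamics (mostly root-relative) and a Control Rig.
# A second AnimDynamics chain is disconnected from the output, and the ClawRot variables are unused.
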AnimBlueprint /BRCosmetics/Accessories/FORT_Backpacks/Backpack_F_MED_SaltLattice/Meshes/F_MED_SaltLattice_Pack_AnimBP
target_skeleton=&/F_MED_SaltLattice_Pack_Skeleton
graphs [
 AnimGraph {nodes [
   n1 = Anim Dynamics - Bone: Fishtail_01(Alpha=1, ExternalForce="0,0,0") {
    angularDampingOverride=1
    angularSpringConstant=300
    bAngularSpring=true
    bOverrideAngularDamping=true
    bUseGravityOverride=true
    linearDampingOverride=0.99
    linearSpringConstant=550
    boundBone {boneName=Fishtail_01}
    componentAppliedLinearAccClamp=<3000 3000 6000>
    componentLinearAccScale=<0.3 0.3 0.4>
    componentLinearVelScale=<0.4 0.4 0.4>
    gravityOverride {z=-980}
    physicsBodyDefinitions [{
      boundBone {boneName=Fishtail_01}
      boxExtents=<10 10 10>
      constraintSetup {angularLimitsMax=<0 30 0>, angularLimitsMin=<0 -30 0>, angularTarget=<1 0 0>}
      localJointOffset=<1.2 0 0>
     }]
   }
   n2 = Anim Dynamics - Bone: Fishtail_02(ComponentPose=@n1, Alpha=1, ExternalForce="0,0,0") {
    angularDampingOverride=1
    angularSpringConstant=300
    bAngularSpring=true
    bOverrideAngularDamping=true
    bUseGravityOverride=true
    linearDampingOverride=0.99
    linearSpringConstant=550
    worldSpaceMinimumScale=0.1
    boundBone {boneName=Fishtail_02}
    componentAppliedLinearAccClamp=<3000 3000 6000>
    componentLinearAccScale=<0.3 0.3 0.4>
    componentLinearVelScale=<0.1 0.1 0.1>
    gravityOverride {z=-980}
    physicsBodyDefinitions [{
      boundBone {boneName=Fishtail_02}
      boxExtents=<10 10 10>
      constraintSetup {angularLimitsMax=<5 20 0>, angularLimitsMin=<-5 -20 0>, angularTarget=<1 0 0>}
      localJointOffset=<1.2 0.1 0.1>
     }]
   }
   n3 = Anim Dynamics - Bone: Fishtail_03(ComponentPose=@n2, Alpha=1, ExternalForce="0,0,0") {
    angularDampingOverride=1
    angularSpringConstant=300
    bAngularSpring=true
    bOverrideAngularDamping=true
    bUseGravityOverride=true
    linearDampingOverride=0.99
    linearSpringConstant=550
    worldSpaceMinimumScale=0.1
    boundBone {boneName=Fishtail_03}
    componentAppliedLinearAccClamp=<3000 3000 6000>
    componentLinearAccScale=<0.3 0.3 0.4>
    componentLinearVelScale=<0.1 0.1 0.1>
    gravityOverride {z=-980}
    physicsBodyDefinitions [{
      boundBone {boneName=Fishtail_03}
      boxExtents=<10 10 10>
      constraintSetup {angularLimitsMax=<10 15 0>, angularLimitsMin=<-10 -15 0>, angularTarget=<1 0 0>}
      localJointOffset=<1.2 0.1 0.1>
     }]
   }
   n4 = Anim Dynamics - Bone: Fishtail_04(ComponentPose=@n3, Alpha=1, ExternalForce="0,0,0") {
    angularDampingOverride=1
    angularSpringConstant=400
    bAngularSpring=true
    bOverrideAngularDamping=true
    bUseGravityOverride=true
    linearDampingOverride=0.99
    linearSpringConstant=550
    worldSpaceMinimumScale=0.1
    boundBone {boneName=Fishtail_04}
    componentAppliedLinearAccClamp=<3000 3000 6000>
    componentLinearAccScale=<0.3 0.3 0.4>
    componentLinearVelScale=<0.1 0.1 0.1>
    gravityOverride {z=-980}
    physicsBodyDefinitions [{
      boundBone {boneName=Fishtail_04}
      boxExtents=<10 10 10>
      constraintSetup {angularLimitsMax=<2 10 0>, angularLimitsMin=<-2 -10 0>, angularTarget=<1 0 0>}
      localJointOffset=<3 0.1 0.1>
     }]
   }
   n5 = Anim Dynamics - Bone: Fishtail_05(ComponentPose=@n4, Alpha=1, ExternalForce="0,0,0") {
    angularDampingOverride=1
    angularSpringConstant=300
    bAngularSpring=true
    bOverrideAngularDamping=true
    bUseGravityOverride=true
    linearDampingOverride=0.99
    linearSpringConstant=550
    worldSpaceMinimumScale=0.1
    boundBone {boneName=Fishtail_05}
    componentAppliedLinearAccClamp=<3000 3000 6000>
    componentLinearAccScale=<0.3 0.3 0.4>
    componentLinearVelScale=<0.1 0.1 0.1>
    gravityOverride {z=-980}
    physicsBodyDefinitions [{
      boundBone {boneName=Fishtail_05}
      boxExtents=<10 10 10>
      constraintSetup {angularLimitsMax=<10 20 0>, angularLimitsMin=<-10 -20 -10>, angularTarget=<1 0 0>}
      localJointOffset=<2 0.1 0.1>
     }]
   }
   n6 = Anim Dynamics - Bone: Fishtail_06(ComponentPose=@n5, Alpha=1, ExternalForce="0,0,0") {
    angularDampingOverride=1
    angularSpringConstant=300
    bAngularSpring=true
    bOverrideAngularDamping=true
    bUseGravityOverride=true
    linearDampingOverride=0.99
    linearSpringConstant=550
    boundBone {boneName=Fishtail_06}
    componentAppliedLinearAccClamp=<3000 3000 6000>
    componentLinearAccScale=<0.3 0.3 0.4>
    componentLinearVelScale=<0.1 0.1 0.1>
    gravityOverride {z=-980}
    physicsBodyDefinitions [{
      boundBone {boneName=Fishtail_06}
      boxExtents=<10 10 10>
      constraintSetup {angularLimitsMax=<0 10 0>, angularLimitsMin=<0 -10 -5>, angularTarget=<1 0 0>}
      localJointOffset=<0.9 0 0>
     }]
   }
   Anim Dynamics - Bone: Fishtail_07(ComponentPose=@n6, Alpha=1, ExternalForce="0,0,0") {
    angularDampingOverride=1
    angularSpringConstant=300
    bAngularSpring=true
    bOverrideAngularDamping=true
    bUseGravityOverride=true
    linearDampingOverride=0.99
    linearSpringConstant=550
    worldSpaceMinimumScale=0.1
    boundBone {boneName=Fishtail_07}
    componentAppliedLinearAccClamp=<3000 3000 6000>
    componentLinearAccScale=<0.3 0.3 0.4>
    componentLinearVelScale=<0.1 0.1 0.1>
    gravityOverride {z=-980}
    physicsBodyDefinitions [{
      boundBone {boneName=Fishtail_07}
      boxExtents=<10 10 10>
      constraintSetup {angularLimitsMax=<5 40 0>, angularLimitsMin=<-5 -40 0>, angularTarget=<1 0 0>}
      localJointOffset=<0.9 0.1 0.1>
     }]
   }
   Component To Local()
   n7 = "Transform (Modify) Bone - Bone: Fishtail_01"(ComponentPose=Local To Component(LocalPose=Input Pose()), Alpha=1, Translation="0,0,0", Rotation="0.000,-8.436,-0.000", Scale="1,1,1") {rotationMode=BMM_Additive, rotationSpace=BCS_BoneSpace, boneToModify {boneName=Fishtail_01}, rotation=<rot 0 -8.4359 0>}
   n8 = "Transform (Modify) Bone - Bone: pack"(ComponentPose=@n7, Translation=Get backpack_translation(), Rotation=Get backpack_rotate()) {rotationMode=BMM_Additive, translationMode=BMM_Additive, alphaBoolBlend {blendOption=Linear}, boneToModify {boneName=pack}}
   n9 = Anim Dynamics - Bone: Fishtail_01(ComponentPose=@n8, Alpha=1, ExternalForce="0,0,0") {
    angularDampingOverride=1
    angularSpringConstant=300
    bAngularSpring=true
    bOverrideAngularDamping=true
    bUseGravityOverride=true
    linearDampingOverride=0.99
    linearSpringConstant=550
    boundBone {boneName=Fishtail_01}
    componentAppliedLinearAccClamp=<3000 3000 6000>
    componentLinearAccScale=<0.5 0.5 0.5>
    componentLinearVelScale=<0.5 0.5 0.5>
    gravityOverride {z=-980}
    physicsBodyDefinitions [{
      boundBone {boneName=Fishtail_01}
      boxExtents=<10 10 10>
      constraintSetup {angularLimitsMax=<0 30 0>, angularLimitsMin=<0 -30 -15>, angularTarget=<1 0 0>}
      localJointOffset=<1.3 0 0>
     }]
   }
   n10 = Anim Dynamics - Bone: Fishtail_02(ComponentPose=@n9, Alpha=1, ExternalForce="0,0,0") {
    angularDampingOverride=1
    angularSpringConstant=300
    bAngularSpring=true
    bOverrideAngularDamping=true
    bUseGravityOverride=true
    linearDampingOverride=0.99
    linearSpringConstant=550
    simulationSpace=RootRelative
    worldSpaceMinimumScale=0.1
    boundBone {boneName=Fishtail_02}
    componentAppliedLinearAccClamp=<3000 3000 6000>
    componentLinearAccScale=<0.5 0.5 0.5>
    componentLinearVelScale=<0.5 0.5 0.5>
    gravityOverride {z=-980}
    physicsBodyDefinitions [{
      boundBone {boneName=Fishtail_02}
      boxExtents=<10 10 10>
      constraintSetup {angularLimitsMax=<5 15 0>, angularLimitsMin=<-5 -15 -20>, angularTarget=<1 0 0>}
      localJointOffset=<1.3 0 0>
     }]
   }
   n11 = Anim Dynamics - Bone: Fishtail_03(ComponentPose=@n10, Alpha=1, ExternalForce="0,0,0") {
    angularDampingOverride=1
    angularSpringConstant=300
    bAngularSpring=true
    bOverrideAngularDamping=true
    bUseGravityOverride=true
    linearDampingOverride=0.99
    linearSpringConstant=550
    simulationSpace=RootRelative
    worldSpaceMinimumScale=0.1
    boundBone {boneName=Fishtail_03}
    componentAppliedLinearAccClamp=<3000 3000 6000>
    componentLinearAccScale=<0.5 0.5 0.5>
    componentLinearVelScale=<0.5 0.5 0.5>
    gravityOverride {z=-980}
    physicsBodyDefinitions [{
      boundBone {boneName=Fishtail_03}
      boxExtents=<10 10 10>
      constraintSetup {angularLimitsMax=<10 15 0>, angularLimitsMin=<-10 -15 -20>, angularTarget=<1 0 0>}
      localJointOffset=<1.3 0 0>
     }]
   }
   n12 = Anim Dynamics - Bone: Fishtail_04(ComponentPose=@n11, Alpha=1, ExternalForce="0,0,0") {
    angularDampingOverride=1
    angularSpringConstant=400
    bAngularSpring=true
    bOverrideAngularDamping=true
    bUseGravityOverride=true
    linearDampingOverride=0.99
    linearSpringConstant=550
    simulationSpace=RootRelative
    worldSpaceMinimumScale=0.1
    boundBone {boneName=Fishtail_04}
    componentAppliedLinearAccClamp=<3000 3000 6000>
    componentLinearAccScale=<0.5 0.5 0.5>
    componentLinearVelScale=<0.5 0.5 0.5>
    gravityOverride {z=-980}
    physicsBodyDefinitions [{
      boundBone {boneName=Fishtail_04}
      boxExtents=<10 10 10>
      constraintSetup {angularLimitsMax=<2 15 0>, angularLimitsMin=<-2 -15 -25>, angularTarget=<1 0 0>}
      localJointOffset=<1.3 0 0>
     }]
   }
   n13 = Anim Dynamics - Bone: Fishtail_05(ComponentPose=@n12, Alpha=1, ExternalForce="0,0,0") {
    angularDampingOverride=1
    angularSpringConstant=300
    bAngularSpring=true
    bOverrideAngularDamping=true
    bUseGravityOverride=true
    linearDampingOverride=0.99
    linearSpringConstant=550
    simulationSpace=RootRelative
    worldSpaceMinimumScale=0.1
    boundBone {boneName=Fishtail_05}
    componentAppliedLinearAccClamp=<3000 3000 6000>
    componentLinearAccScale=<0.5 0.5 0.5>
    componentLinearVelScale=<0.5 0.5 0.5>
    gravityOverride {z=-980}
    physicsBodyDefinitions [{
      boundBone {boneName=Fishtail_05}
      boxExtents=<10 10 10>
      constraintSetup {angularLimitsMax=<10 15 0>, angularLimitsMin=<-10 -15 -25>, angularTarget=<1 0 0>}
      localJointOffset=<1.3 0 0>
     }]
   }
   n14 = Anim Dynamics - Bone: Fishtail_06(ComponentPose=@n13, Alpha=1, ExternalForce="0,0,0") {
    angularDampingOverride=1
    angularSpringConstant=300
    bAngularSpring=true
    bOverrideAngularDamping=true
    bUseGravityOverride=true
    linearDampingOverride=0.99
    linearSpringConstant=550
    simulationSpace=RootRelative
    boundBone {boneName=Fishtail_06}
    componentAppliedLinearAccClamp=<3000 3000 6000>
    componentLinearAccScale=<0.5 0.5 0.5>
    componentLinearVelScale=<0.5 0.5 0.5>
    gravityOverride {z=-980}
    physicsBodyDefinitions [{
      boundBone {boneName=Fishtail_06}
      boxExtents=<10 10 10>
      constraintSetup {angularLimitsMax=<0 15 0>, angularLimitsMin=<0 -15 -25>, angularTarget=<1 0 0>}
      localJointOffset=<1.3 0 0>
     }]
   }
   n15 = Anim Dynamics - Bone: Fishtail_07(ComponentPose=@n14, Alpha=1, ExternalForce="0,0,0") {
    angularDampingOverride=1
    angularSpringConstant=300
    bAngularSpring=true
    bOverrideAngularDamping=true
    bUseGravityOverride=true
    linearDampingOverride=0.99
    linearSpringConstant=550
    simulationSpace=RootRelative
    worldSpaceMinimumScale=0.1
    boundBone {boneName=Fishtail_07}
    componentAppliedLinearAccClamp=<3000 3000 6000>
    componentLinearAccScale=<0.5 0.5 0.5>
    componentLinearVelScale=<0.5 0.5 0.5>
    gravityOverride {z=-980}
    physicsBodyDefinitions [{
      boundBone {boneName=Fishtail_07}
      boxExtents=<10 10 10>
      constraintSetup {angularLimitsMax=<5 15 0>, angularLimitsMin=<-5 -15 -25>, angularTarget=<1 0 0>}
      localJointOffset=<1.3 0 0>
     }]
   }
   n16 = Control Rig(Source=Component To Local(ComponentPose=@n15)) {controlRigAssetReference {blueprintRigClass=&/F_MED_SaltLattice_Pack_CtrlRig.F_MED_SaltLattice_Pack_CtrlRig_C}}
   Output Pose(Result=@n16) {layerGroup=None}
  ]}
 EventGraph {nodes [
   Event BlueprintUpdateAnimation()
   n1 = GetOwnerSkeletalMeshComponent()
   n2 = BreakRotator(InRot=BreakTransform(InTransform=GetDeltaTransformFromRefPose(self=@n1, BoneName=spine_05, BaseName=spine_01)).Rotation)
   n3 = MapRangeClamped(Value=@n2.Yaw, InRangeA=-10, InRangeB=-50, OutRangeB=50)
   > Set backpack_rotate(backpack_rotate=MakeRotator(Roll=@n3, Pitch=float / float(A=@n3, B=-10)))
   > Set backpack_translation(backpack_translation=MakeVector(Y=MapRangeClamped(Value=@n2.Yaw, InRangeA=-15, InRangeB=-50, OutRangeB=-9), Z=3))
   TryGetPawnOwner()
  ]}
]
variables [
 backpack_rotate {type=Rotator}
 backpack_translation {type=Vector}
 ClawRot1 {type=Rotator}
 ClawRot2 {type=Rotator}
 ClawRot3 {type=Rotator}
]
